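# Physics asset for the MissusMind Goop female medium head cosmetic, providing bodies and joints for eight dangling goo strands (A–H).
# Each strand pairs a small master capsule, jointed to the head capsule with free swing and twist, with a longer offset sim capsule.
# Sim-to-master joints carry per-strand swing limits, tightest on D–F, and very narrow twist limits, so strands can sway with little roll.
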PhysicsAsset /BRCosmetics/Characters/Player/Female/Medium/Heads/F_MED_MissusMind_Goop_Head/Meshes/F_MED_MissusMind_Goop_Head_Physics
bodies {
 dyn_head_goo_A_mstr [{shapeType=Capsule, length=0.505, radius=0.505, shapeName=dyn_faceAcc_goop_A_mstr_capsule, rotation=<rot -90 0 0>}]
 dyn_head_goo_A_sim [{shapeType=Capsule, length=7.0333, radius=1.1699, shapeName=dyn_faceAcc_goop_A_sim_capsule, center=<4.4367 0 0>, rotation=<rot -90 180 180>}]
 dyn_head_goo_B_mstr [{shapeType=Capsule, length=0.505, radius=0.505, shapeName=dyn_faceAcc_goop_A_mstr_capsule, rotation=<rot -90 0 0>}]
 dyn_head_goo_B_sim [{shapeType=Capsule, length=7.0333, radius=1.1699, shapeName=dyn_faceAcc_goop_A_sim_capsule, center=<4.4367 0 0>, rotation=<rot -90 180 180>}]
 dyn_head_goo_C_mstr [{shapeType=Capsule, length=0.505, radius=0.505, shapeName=dyn_faceAcc_goop_A_mstr_capsule, rotation=<rot -90 0 0>}]
 dyn_head_goo_C_sim [{shapeType=Capsule, length=7.0333, radius=1.1699, shapeName=dyn_faceAcc_goop_A_sim_capsule, center=<4.4367 0 0>, rotation=<rot -90 180 180>}]
 dyn_head_goo_D_mstr [{shapeType=Capsule, length=0.505, radius=0.505, shapeName=dyn_faceAcc_goop_A_mstr_capsule, rotation=<rot -90 0 0>}]
 dyn_head_goo_D_sim [{shapeType=Capsule, length=7.0333, radius=1.1699, shapeName=dyn_faceAcc_goop_A_sim_capsule, center=<4.4367 0 0>, rotation=<rot -90 180 180>}]
 dyn_head_goo_E_mstr [{shapeType=Capsule, length=0.505, radius=0.505, shapeName=dyn_faceAcc_goop_A_mstr_capsule, rotation=<rot -90 0 0>}]
 dyn_head_goo_E_sim [{shapeType=Capsule, length=7.0333, radius=1.1699, shapeName=dyn_faceAcc_goop_A_sim_capsule, center=<4.4367 0 0>, rotation=<rot -90 180 180>}]
 dyn_head_goo_F_mstr [{shapeType=Capsule, length=0.505, radius=0.505, shapeName=dyn_faceAcc_goop_A_mstr_capsule, rotation=<rot -90 0 0>}]
 dyn_head_goo_F_sim [{shapeType=Capsule, length=7.0333, radius=1.1699, shapeName=dyn_faceAcc_goop_A_sim_capsule, center=<4.4367 0 0>, rotation=<rot -90 180 180>}]
 dyn_head_goo_G_mstr [{shapeType=Capsule, length=0.505, radius=0.505, shapeName=dyn_faceAcc_goop_A_mstr_capsule, rotation=<rot -90 0 0>}]
 dyn_head_goo_G_sim [{shapeType=Capsule, length=7.0333, radius=1.1699, shapeName=dyn_faceAcc_goop_A_sim_capsule, center=<4.4367 0 0>, rotation=<rot -90 180 180>}]
 dyn_head_goo_H_mstr [{shapeType=Capsule, length=0.505, radius=0.505, shapeName=dyn_faceAcc_goop_A_mstr_capsule, rotation=<rot -90 0 0>}]
 dyn_head_goo_H_sim [{shapeType=Capsule, length=7.0333, radius=1.1699, shapeName=dyn_faceAcc_goop_A_sim_capsule, center=<4.4367 0 0>, rotation=<rot -90 180 180>}]
 head [{shapeType=Capsule, length=9.4508, radius=2.2986, shapeName=head_capsule, rotation=<rot -90 0 0>}]
}
constraints [
 {bone1Name=dyn_head_goo_A_sim, bone2Name=dyn_head_goo_A_mstr, swing1LimitDegrees=40, swing2LimitDegrees=35, twistLimitDegrees=2}
 {bone1Name=dyn_head_goo_B_sim, bone2Name=dyn_head_goo_B_mstr, swing1LimitDegrees=40, swing2LimitDegrees=35, twistLimitDegrees=2}
 {bone1Name=dyn_head_goo_C_sim, bone2Name=dyn_head_goo_C_mstr, swing1LimitDegrees=35, swing2LimitDegrees=35, twistLimitDegrees=2}
 {bone1Name=dyn_head_goo_D_sim, bone2Name=dyn_head_goo_D_mstr, swing1LimitDegrees=20, swing2LimitDegrees=20, twistLimitDegrees=2}
 {bone1Name=dyn_head_goo_E_sim, bone2Name=dyn_head_goo_E_mstr, swing1LimitDegrees=20, swing2LimitDegrees=20, twistLimitDegrees=2}
 {bone1Name=dyn_head_goo_F_sim, bone2Name=dyn_head_goo_F_mstr, swing1LimitDegrees=20, swing2LimitDegrees=20, twistLimitDegrees=2}
 {bone1Name=dyn_head_goo_G_sim, bone2Name=dyn_head_goo_G_mstr, swing1LimitDegrees=35, swing2LimitDegrees=45, twistLimitDegrees=2}
 {bone1Name=dyn_head_goo_H_sim, bone2Name=dyn_head_goo_H_mstr, swing1LimitDegrees=35, swing2LimitDegrees=45, twistLimitDegrees=2}
 {
  bone1Name=dyn_head_goo_A_mstr
  bone2Name=head
  swing1LimitDegrees=45
  swing1Motion=Free
  swing2LimitDegrees=45
  swing2Motion=Free
  twistLimitDegrees=45
  twistMotion=Free
 }
 {
  bone1Name=dyn_head_goo_B_mstr
  bone2Name=head
  swing1LimitDegrees=45
  swing1Motion=Free
  swing2LimitDegrees=45
  swing2Motion=Free
  twistLimitDegrees=45
  twistMotion=Free
 }
 {
  bone1Name=dyn_head_goo_C_mstr
  bone2Name=head
  swing1LimitDegrees=45
  swing1Motion=Free
  swing2LimitDegrees=45
  swing2Motion=Free
  twistLimitDegrees=45
  twistMotion=Free
 }
 {
  bone1Name=dyn_head_goo_D_mstr
  bone2Name=head
  swing1LimitDegrees=45
  swing1Motion=Free
  swing2LimitDegrees=45
  swing2Motion=Free
  twistLimitDegrees=45
  twistMotion=Free
 }
 {
  bone1Name=dyn_head_goo_E_mstr
  bone2Name=head
  swing1LimitDegrees=45
  swing1Motion=Free
  swing2LimitDegrees=45
  swing2Motion=Free
  twistLimitDegrees=45
  twistMotion=Free
 }
 {
  bone1Name=dyn_head_goo_F_mstr
  bone2Name=head
  swing1LimitDegrees=45
  swing1Motion=Free
  swing2LimitDegrees=45
  swing2Motion=Free
  twistLimitDegrees=45
  twistMotion=Free
 }
 {
  bone1Name=dyn_head_goo_G_mstr
  bone2Name=head
  swing1LimitDegrees=45
  swing1Motion=Free
  swing2LimitDegrees=45
  swing2Motion=Free
  twistLimitDegrees=45
  twistMotion=Free
 }
 {
  bone1Name=dyn_head_goo_H_mstr
  bone2Name=head
  swing1LimitDegrees=45
  swing1Motion=Free
  swing2LimitDegrees=45
  swing2Motion=Free
  twistLimitDegrees=45
  twistMotion=Free
 }
]
